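# Niagara effect for the SurgeRaven player emote that throws tumbling candy sprites from skeletal mesh bones or sockets.
# Candy_Burst fires timed bursts while three rate emitters release curve-driven streams, all tinted by time of day.
# Spawn groups select attach point, velocity and renderer visibility tag within shared emitters.
# Particles fall under ramped gravity, with quality-gated collision modules; the rate emitters force a local analytical plane in front-end preview.
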NiagaraSystem /BRCosmetics/Animation/Game/MainPlayer/Emotes/SurgeRaven/Fx/NS_Emote_SurgeRaven_Candy
dependencies {
 usedDataInterfaces [&NiagaraDataInterfaceCurve, &NiagaraDataInterfacePlatformSet, &NiagaraDataInterfaceSkeletalMesh]
 usedDynamicInputs [&FNTimeOfDayColor]
 usedModules [
  &self:SystemSpawnScript.SystemScriptSource.SystemScriptGraph.NiagaraNodeAssignment_0.SetVariables_#1
  &SystemState
  &self:SystemSpawnScript.SystemScriptSource.SystemScriptGraph.NiagaraNodeAssignment_1.SetVariables_#2
 ]
}
emitters {
 Candy_Burst {
  emitterUpdateScript [
   &EmitterState(Life Cycle Mode=::NewEnumerator1, Loop Behavior=::NewEnumerator1, Loop Duration=8, UseLoopDelay=false)
   &SpawnBurst_Instantaneous(
    Spawn Count=&RandomRangeInt(Minimum=12, Maximum=24)
    Use Spawn Probability=false
    UseLoopCountLimit=false
   )
   &SpawnBurst_Instantaneous(
    Spawn Count=&RandomRangeInt(Minimum=8, Maximum=16)
    Spawn Group=1
    Use Spawn Probability=false
    UseLoopCountLimit=false
   ) {module=SpawnBurst_Instantaneous004}
   &SpawnBurst_Instantaneous(
    Spawn Count=&RandomRangeInt(Minimum=12, Maximum=24)
    Spawn Group=2
    Spawn Time=4.57
    Use Spawn Probability=false
    UseLoopCountLimit=false
   ) {enabled=false, module=SpawnBurst_Instantaneous008}
   &SpawnBurst_Instantaneous(
    Spawn Count=&RandomRangeInt(Minimum=8, Maximum=12)
    Spawn Group=3
    Spawn Time=0.36
    Use Spawn Probability=false
    UseLoopCountLimit=false
   ) {module=SpawnBurst_Instantaneous001}
   &SpawnBurst_Instantaneous(
    Spawn Count=&RandomRangeInt(Minimum=8, Maximum=12)
    Spawn Group=4
    Spawn Time=4.9
    Use Spawn Probability=false
    UseLoopCountLimit=false
   ) {module=SpawnBurst_Instantaneous002}
   &SpawnBurst_Instantaneous(
    Spawn Count=&RandomRangeInt(Minimum=4, Maximum=7)
    Spawn Group=5
    Spawn Time=3.35
    Use Spawn Probability=false
    UseLoopCountLimit=false
   ) {module=SpawnBurst_Instantaneous003}
   &SpawnBurst_Instantaneous(
    Spawn Count=&RandomRangeInt(Minimum=8, Maximum=12)
    Spawn Group=6
    Spawn Time=0.575
    Use Spawn Probability=false
    UseLoopCountLimit=false
   ) {module=SpawnBurst_Instantaneous005}
   &SpawnBurst_Instantaneous(
    Spawn Count=&RandomRangeInt(Minimum=8, Maximum=12)
    Spawn Group=7
    Spawn Time=5.04
    Use Spawn Probability=false
    UseLoopCountLimit=false
   ) {module=SpawnBurst_Instantaneous006}
  ]
  particleSpawnScript [
   /SetVariables(Particles.VisibilityTag=&SpawnGroup()) {isSetParameters=true, module=SetVariables_#3}
   &InitializeParticle(
    Color=$System.ToD
    Lifetime=&MergeHelper_FloatSwitch(
     0=&RandomRangeFloat(Minimum=5, Maximum=5.8)
     1=&RandomRangeFloat(Minimum=5, Maximum=5.8)
     2=&RandomRangeFloat(Minimum=5, Maximum=5.8)
     3=&RandomRangeFloat(Minimum=5, Maximum=5.8)
     4=&RandomRangeFloat(Minimum=5, Maximum=5.8)
     5=&RandomRangeFloat(Minimum=5, Maximum=5.8)
     6=&RandomRangeFloat(Minimum=5, Maximum=5.8)
     7=&RandomRangeFloat(Minimum=5, Maximum=5.8)
     Index=&SpawnGroup()
     Num=::NewEnumerator6
    )
    Sprite Rotation Mode=::NewEnumerator2
    Sprite Size Mode=::NewEnumerator1
    Uniform Sprite Size Max=9
    Uniform Sprite Size Min=7
    UsePositionOffset=false
    Write Material Random=true
   )
   &ShapeLocation(Handle Radius=2, Large Radius=4, MaskSpawnGroup=false, Override Seed=false, Shape Primitive=::NewEnumerator3, UseOwnerScale=true)
   &SkeletalMeshLocation(
    Bone / Socket Index=&MergeHelper_IntSwitch(0=0, 1=0, 2=2, 3=0, 4=2, 5=1, 6=3, 7=4, Index=&SpawnGroup(), Num=::NewEnumerator6)
    Bone Sampling Mode=::NewEnumerator4
    MaskSpawnGroup=false
    Sampled Position Offset=&AddVectorToPosition(
     Position=$Particles.ShapeLocation.ShapePosition
     Vector=&MakeVector(
      X=0
      Y=0
      Z=&MergeHelper_FloatSwitch(0=5, 1=5, 2=2, 3=2, 4=2, 5=0, 6=2, 7=2, Index=&SpawnGroup(), Num=::NewEnumerator6)
     )
    )
    Skeletal Mesh=$System.Skeletal Mesh Reader
   )
   &AddVelocity(
    Module.Use Falloff Distance=false
    Origin Offset:NiagaraPosition=$Particles.SkeletalMeshLocation.SampledPosition
    Override Seed=false
    Velocity Mode=::NewEnumerator1
    Velocity Speed=&MergeHelper_FloatSwitch(
     0=0
     1=0
     2=&RandomRangeFloat(Minimum=125, Maximum=450)
     3=0
     4=0
     5=0
     6=0
     7=&RandomRangeFloat(Minimum=45, Maximum=90)
     Index=&SpawnGroup()
     Num=::NewEnumerator6
    )
   ) {module=AddVelocity001}
   &AddVelocity(
    Cone Angle=&MergeHelper_FloatSwitch(0=35, 1=90, 2=0, 3=120, 4=120, 5=0, 6=25, 7=0, Index=&SpawnGroup(), Num=::NewEnumerator6)
    Cone Axis=&MakeVector(
     X=&MergeHelper_FloatSwitch(0=1, 1=1, 2=0, 3=-1, 4=1, 5=-1, 6=-1, 7=0, Index=&SpawnGroup(), Num=::NewEnumerator6)
     Y=0
     Z=&MergeHelper_FloatSwitch(0=0.1, 1=0.2, 2=0, 3=0.1, 4=0, 5=0, 6=0, 7=0, Index=&SpawnGroup(), Num=::NewEnumerator6)
    )
    Module.UseConeFalloff=false
    Module.UseVelDistribution=false
    Override Seed=false
    Velocity Mode=::NewEnumerator2
    Velocity Speed=&MergeHelper_FloatSwitch(
     0=&RandomRangeFloat(Minimum=200, Maximum=525)
     1=&RandomRangeFloat(Minimum=125, Maximum=400)
     2=0
     3=&RandomRangeFloat(Minimum=30, Maximum=125)
     4=&RandomRangeFloat(Minimum=80, Maximum=165)
     5=&RandomRangeFloat(Minimum=30, Maximum=125)
     6=&RandomRangeFloat(Minimum=5, Maximum=35)
     7=0
     Index=&SpawnGroup()
     Num=::NewEnumerator6
    )
   )
   &SubUVAnimation(End Frame=3, SubUV Animation Mode:ENiagara_SubUVLookupMode=::NewEnumerator2)
  ]
  particleUpdateScript [
   &ParticleState()
   &SpriteRotationRate(Rotation Rate=&FloatFromCurve(
     FloatCurve=DataInterface() {propertyValues {Curve {keys [<at 0 1 RCIM_Cubic>, {arriveTangent=-3.1321, interpMode=RCIM_Cubic, leaveTangent=-3.1321, tangentMode=RCTM_User, time=0.2, value=0}]}}}
     Scale Curve=&MergeHelper_FloatSwitch(
      0=&RandomRangeFloat(Minimum=1750, Maximum=2500)
      1=&RandomRangeFloat(Minimum=1750, Maximum=2500)
      2=&RandomRangeFloat(Minimum=-1750, Maximum=1750)
      3=&RandomRangeFloat(Minimum=-1750, Maximum=1750)
      4=&RandomRangeFloat(Minimum=-1750, Maximum=1750)
      5=&RandomRangeFloat(Minimum=-1750, Maximum=1750)
      6=&RandomRangeFloat(Minimum=-1750, Maximum=-2500)
      7=&RandomRangeFloat(Minimum=-1750, Maximum=1750)
      Index=&SpawnGroup()
      Num=::NewEnumerator6
     )
    ))
   &ScaleSpriteSize(Uniform Curve Sprite Scale=DataInterface() {propertyValues {Curve {keys [
        {tangentMode=RCTM_User, time=0, value=1}
        {arriveTangent=0.2105, interpMode=RCIM_Cubic, leaveTangent=0.2105, time=0.9, value=1}
        {arriveTangent=-10, interpMode=RCIM_Cubic, leaveTangent=-10, time=0.95, value=1.2}
        <at 1 0 RCIM_Cubic>
       ]}}})
   &GravityForce(Gravity=&MakeVector(
     X=0
     Y=0
     Z=&FloatFromCurve(FloatCurve=DataInterface() {propertyValues {Curve {keys [<at 0.05 0 RCIM_Cubic>,<at 0.1 1 RCIM_Cubic>]}}}, Scale Curve=-980)
    ))
   &Collision(
    Age Colliding Particles=false
    Collision Enabled=&IsPlatformSetActive(New Platform Set=DataInterface() {propertyValues {Platforms {qualityLevelMask=-8}}})
    Control Roll On Collision=true
    Enable Rest State=false
    KillOnCollision=false
    Module.EnableMaxCollisionCount=false
    Module.Randomize Collision Normal Vector=false
    Restitution=0.2
    ScaleRateOfRotation=true
   ) {module=Collision001}
   &Collision(
    Age Colliding Particles=false
    Analytical Collision Plane Position 1:Vector3f=<0 0 -125>
    Analytical Collision Plane Space:ENiagaraCoordinateSpace=::Local
    CPU Collision Type=::NewEnumerator3
    Collision Enabled=&IsPlatformSetActive(New Platform Set=DataInterface() {propertyValues {Platforms {qualityLevelMask=-28}}})
    Control Roll On Collision=true
    Enable Rest State=false
    KillOnCollision=false
    Module.EnableMaxCollisionCount=false
    Module.Randomize Collision Normal Vector=false
    Particle Radius Scale=5
    Perform 2 Analytical Collision Queries=false
    Restitution=0.2
    ScaleRateOfRotation=false
   ) {module=Collision002}
   &Drag(Drag=&RandomRangeFloat(Minimum=0.8, Maximum=1.2), Use Linear Drag=true, Use Rotational Drag=true)
   &SolveForcesAndVelocity(Clamp Velocity=false, Limit Acceleration=false)
  ]
  rendererClasses [&NiagaraSpriteRendererProperties]
  renderers [
   {rendererClass=&NiagaraSpriteRendererProperties}
   {rendererClass=&NiagaraSpriteRendererProperties, rendererIndex=1}
   {rendererClass=&NiagaraSpriteRendererProperties, rendererIndex=2}
   {rendererClass=&NiagaraSpriteRendererProperties, rendererIndex=3}
   {rendererClass=&NiagaraSpriteRendererProperties, rendererIndex=4}
   {rendererClass=&NiagaraSpriteRendererProperties, rendererIndex=5}
   {rendererClass=&NiagaraSpriteRendererProperties, rendererIndex=6}
   {rendererClass=&NiagaraSpriteRendererProperties, rendererIndex=7}
  ]
 }
 Candy_Head_Rate {
  emitterUpdateScript [
   &EmitterState(Life Cycle Mode=::NewEnumerator1, Loop Behavior=::NewEnumerator1, Loop Delay=3.1, Loop Duration=2.3, UseLoopDelay=true)
   &SpawnRate(
    SpawnRate=&FloatFromCurve(
     FloatCurve=DataInterface() {propertyValues {Curve {keys [
         <at 0 1.7 RCIM_Cubic>
         {arriveTangent=-6.73, interpMode=RCIM_Cubic, leaveTangent=-6.73, tangentMode=RCTM_User, time=0.225, value=1}
         {arriveTangent=-0.229, interpMode=RCIM_Cubic, leaveTangent=-0.229, tangentMode=RCTM_User, time=0.7, value=0}
        ]}}}
     CurveIndex=$Emitter.NormalizedLoopAge
     Scale Curve=60
    )
    Use Spawn Probability=false
   )
  ]
  particleSpawnScript [
   /SetVariables(Particles.VisibilityTag=&SpawnGroup()) {isSetParameters=true, module=SetVariables_#3}
   &InitializeParticle(
    Color=$System.ToD
    Lifetime Max=3
    Lifetime Min=2.4
    Lifetime Mode=::NewEnumerator1
    Sprite Rotation Mode=::NewEnumerator2
    Sprite Size Mode=::NewEnumerator1
    Uniform Sprite Size Max=9
    Uniform Sprite Size Min=7
    UsePositionOffset=false
    Write Material Random=true
   )
   &ShapeLocation(Handle Radius=2, Large Radius=4, MaskSpawnGroup=false, Override Seed=false, Shape Primitive=::NewEnumerator3, UseOwnerScale=true) {enabled=false}
   &SkeletalMeshLocation(
    Bone / Socket Index=5
    Bone Sampling Mode=::NewEnumerator4
    MaskSpawnGroup=false
    Sampled Position Offset=<3 -7 4>
    Sampled Position Offset Coordinate Space:ENiagaraSkelMeshTransforms=::NewEnumerator2
    Skeletal Mesh=$System.Skeletal Mesh Reader
   )
   &AddVelocity(
    Cone Angle=25
    Cone Axis=<0 0 1>
    Module.UseConeFalloff=false
    Module.UseVelDistribution=false
    Override Seed=false
    Velocity Mode=::NewEnumerator2
    Velocity Speed=&FloatFromCurve(
     FloatCurve=DataInterface() {propertyValues {Curve {keys [
         <at 0 2.25 RCIM_Cubic>
         {arriveTangent=-0.0852, interpMode=RCIM_Cubic, leaveTangent=-0.0852, tangentMode=RCTM_User, time=0.2, value=1}
         {arriveTangent=-0.0449, interpMode=RCIM_Cubic, leaveTangent=-0.0449, tangentMode=RCTM_User, time=1, value=0}
        ]}}}
     CurveIndex=$Emitter.NormalizedLoopAge
     Scale Curve=&RandomRangeFloat(Minimum=65, Maximum=175)
    )
   )
   &SubUVAnimation(End Frame=3, SubUV Animation Mode:ENiagara_SubUVLookupMode=::NewEnumerator2)
  ]
  particleUpdateScript [
   &ParticleState()
   &SpriteRotationRate(Rotation Rate=&FloatFromCurve(
     FloatCurve=DataInterface() {propertyValues {Curve {keys [<at 0 1 RCIM_Cubic>, {arriveTangent=-3.1321, interpMode=RCIM_Cubic, leaveTangent=-3.1321, tangentMode=RCTM_User, time=0.6, value=0}]}}}
     Scale Curve=&RandomRangeFloat(Minimum=-2500, Maximum=2500)
    ))
   &ScaleSpriteSize(Uniform Curve Sprite Scale=DataInterface() {propertyValues {Curve {keys [
        {tangentMode=RCTM_User, time=0, value=0.4}
        <at 0.1 1>
        {arriveTangent=0.2353, interpMode=RCIM_Cubic, leaveTangent=0.2353, time=0.9, value=1}
        {arriveTangent=-10, interpMode=RCIM_Cubic, leaveTangent=-10, time=0.95, value=1.2}
        <at 1 0 RCIM_Cubic>
       ]}}})
   &GravityForce(Gravity=&MakeVector(
     X=0
     Y=0
     Z=&FloatFromCurve(FloatCurve=DataInterface() {propertyValues {Curve {keys [<at 0.05 0 RCIM_Cubic>,<at 0.1 1 RCIM_Cubic>]}}}, Scale Curve=-980)
    ))
   &Collision(
    Age Colliding Particles=false
    Collision Enabled=&Boolean_AndOperation(
     Bool1=&IsPlatformSetActive(New Platform Set=DataInterface() {propertyValues {Platforms {qualityLevelMask=-8}}})
     Bool2=&Boolean_NotOperation(Bool=$User.bIsFrontEndPreview)
    )
    Control Roll On Collision=true
    Enable Rest State=false
    KillOnCollision=false
    Module.EnableMaxCollisionCount=false
    Module.Randomize Collision Normal Vector=false
    Restitution=0.2
    ScaleRateOfRotation=true
   ) {module=Collision001}
   &Collision(
    Age Colliding Particles=false
    Analytical Collision Plane Position 1:Vector3f=<0 0 -125>
    Analytical Collision Plane Space:ENiagaraCoordinateSpace=::Local
    CPU Collision Type=::NewEnumerator3
    Collision Enabled=&Boolean_OrOperation(
     Bool1=&IsPlatformSetActive(New Platform Set=DataInterface() {propertyValues {Platforms {qualityLevelMask=-28}}})
     Bool2=$User.bIsFrontEndPreview
    )
    Control Roll On Collision=true
    Enable Rest State=false
    KillOnCollision=false
    Module.EnableMaxCollisionCount=false
    Module.Randomize Collision Normal Vector=false
    Particle Radius Scale=5
    Perform 2 Analytical Collision Queries=false
    Restitution=0.2
    ScaleRateOfRotation=false
   ) {module=Collision002}
   &Drag(Drag=&RandomRangeFloat(Minimum=0.8, Maximum=1.2), Use Linear Drag=true, Use Rotational Drag=true)
   &SolveForcesAndVelocity(Clamp Velocity=false, Limit Acceleration=false)
  ]
  rendererClasses [&NiagaraSpriteRendererProperties]
  renderers [{rendererClass=&NiagaraSpriteRendererProperties}]
 }
 Candy_Rate001 {
  emitterUpdateScript [
   &EmitterState(Life Cycle Mode=::NewEnumerator1, Loop Behavior=::NewEnumerator1, Loop Duration=2.3, UseLoopDelay=true)
   &SpawnRate(
    SpawnRate=&FloatFromCurve(
     FloatCurve=DataInterface() {propertyValues {Curve {keys [<at 0 1 RCIM_Cubic>, {arriveTangent=-12.0168, interpMode=RCIM_Cubic, leaveTangent=-12.0168, tangentMode=RCTM_User, time=0.2, value=0}]}}}
     CurveIndex=$Emitter.NormalizedLoopAge
     Scale Curve=56
    )
    Use Spawn Probability=false
   )
   &SpawnRate(
    SpawnGroup=1
    SpawnRate=&FloatFromCurve(
     FloatCurve=DataInterface() {propertyValues {Curve {keys [<at 0 1 RCIM_Cubic>, {arriveTangent=-12.0168, interpMode=RCIM_Cubic, leaveTangent=-12.0168, tangentMode=RCTM_User, time=0.2, value=0}]}}}
     CurveIndex=$Emitter.NormalizedLoopAge
     Scale Curve=56
    )
    Use Spawn Probability=false
   ) {module=SpawnRate001}
  ]
  particleSpawnScript [
   /SetVariables(Particles.VisibilityTag=&SpawnGroup()) {isSetParameters=true, module=SetVariables_#3}
   &InitializeParticle(
    Color=$System.ToD
    Lifetime Max=5.8
    Lifetime Min=5
    Lifetime Mode=::NewEnumerator1
    Sprite Rotation Mode=::NewEnumerator2
    Sprite Size Mode=::NewEnumerator1
    Uniform Sprite Size Max=9
    Uniform Sprite Size Min=7
    UsePositionOffset=false
    Write Material Random=true
   )
   &ShapeLocation(Handle Radius=2, Large Radius=4, MaskSpawnGroup=false, Override Seed=false, Shape Primitive=::NewEnumerator3, UseOwnerScale=true)
   &SkeletalMeshLocation(
    Bone / Socket Index=&MergeHelper_IntSwitch(0=0, 1=3, 2=?, 3=?, 4=?, 5=?, Index=&SpawnGroup(), Num=::NewEnumerator0)
    Bone Sampling Mode=::NewEnumerator4
    MaskSpawnGroup=false
    Sampled Position Offset=&AddVectorToPosition(
     Position=$Particles.ShapeLocation.ShapePosition
     Vector=&MakeVector(
      X=0
      Y=0
      Z=&MergeHelper_FloatSwitch(0=3, 1=-3, 2=0, 3=0, 4=0, Index=&SpawnGroup(), Num=::NewEnumerator3)
     )
    )
    Skeletal Mesh=$System.Skeletal Mesh Reader
   )
   &AddVelocity(
    Cone Angle=60
    Cone Axis=&MakeVector(
     X=&MergeHelper_FloatSwitch(0=1, 1=0, 2=0, 3=0, 4=0, Index=&SpawnGroup(), Num=::NewEnumerator3)
     Y=0
     Z=&MergeHelper_FloatSwitch(0=1, 1=-1, 2=0, 3=0, 4=0, Index=&SpawnGroup(), Num=::NewEnumerator3)
    )
    Module.UseConeFalloff=false
    Module.UseVelDistribution=false
    Override Seed=false
    Velocity Mode=::NewEnumerator2
    Velocity Speed=&MergeHelper_FloatSwitch(
     0=&RandomRangeFloat(Minimum=80, Maximum=120)
     1=&RandomRangeFloat(Minimum=30, Maximum=75)
     2=?
     3=?
     Index=&SpawnGroup()
     Num=::NewEnumerator0
    )
   )
   &SubUVAnimation(End Frame=3, SubUV Animation Mode:ENiagara_SubUVLookupMode=::NewEnumerator2)
  ]
  particleUpdateScript [
   &ParticleState()
   &SpriteRotationRate(Rotation Rate=&FloatFromCurve(
     FloatCurve=DataInterface() {propertyValues {Curve {keys [<at 0 1 RCIM_Cubic>, {arriveTangent=-3.1321, interpMode=RCIM_Cubic, leaveTangent=-3.1321, tangentMode=RCTM_User, time=0.225, value=0}]}}}
     Scale Curve=&RandomRangeFloat(Minimum=-2500, Maximum=2500)
    ))
   &ScaleSpriteSize(Uniform Curve Sprite Scale=DataInterface() {propertyValues {Curve {keys [
        {tangentMode=RCTM_User, time=0, value=1}
        {arriveTangent=0.2105, interpMode=RCIM_Cubic, leaveTangent=0.2105, time=0.9, value=1}
        {arriveTangent=-10, interpMode=RCIM_Cubic, leaveTangent=-10, time=0.95, value=1.2}
        <at 1 0 RCIM_Cubic>
       ]}}})
   &GravityForce(Gravity=&MakeVector(
     X=0
     Y=0
     Z=&FloatFromCurve(FloatCurve=DataInterface() {propertyValues {Curve {keys [<at 0.05 0 RCIM_Cubic>,<at 0.1 1 RCIM_Cubic>]}}}, Scale Curve=-980)
    ))
   &Collision(
    Age Colliding Particles=false
    Collision Enabled=&Boolean_AndOperation(
     Bool1=&IsPlatformSetActive(New Platform Set=DataInterface() {propertyValues {Platforms {qualityLevelMask=-8}}})
     Bool2=&Boolean_NotOperation(Bool=$User.bIsFrontEndPreview)
    )
    Control Roll On Collision=true
    Enable Rest State=false
    KillOnCollision=false
    Module.EnableMaxCollisionCount=false
    Module.Randomize Collision Normal Vector=false
    Restitution=0.2
    ScaleRateOfRotation=true
   ) {module=Collision001}
   &Collision(
    Age Colliding Particles=false
    Analytical Collision Plane Position 1:Vector3f=<0 0 -125>
    Analytical Collision Plane Space:ENiagaraCoordinateSpace=::Local
    CPU Collision Type=::NewEnumerator3
    Collision Enabled=&Boolean_OrOperation(
     Bool1=&IsPlatformSetActive(New Platform Set=DataInterface() {propertyValues {Platforms {qualityLevelMask=-28}}})
     Bool2=$User.bIsFrontEndPreview
    )
    Control Roll On Collision=true
    Enable Rest State=false
    KillOnCollision=false
    Module.EnableMaxCollisionCount=false
    Module.Randomize Collision Normal Vector=false
    Particle Radius Scale=5
    Perform 2 Analytical Collision Queries=false
    Restitution=0.2
    ScaleRateOfRotation=false
   ) {module=Collision002}
   &Drag(Drag=&RandomRangeFloat(Minimum=0.8, Maximum=1.2), Use Linear Drag=true, Use Rotational Drag=true)
   &SolveForcesAndVelocity(Clamp Velocity=false, Limit Acceleration=false)
  ]
  rendererClasses [&NiagaraSpriteRendererProperties]
  renderers [
   {rendererClass=&NiagaraSpriteRendererProperties}
   {rendererClass=&NiagaraSpriteRendererProperties, rendererIndex=1}
  ]
 }
 Candy_Rate002 {
  emitterUpdateScript [
   &EmitterState(Life Cycle Mode=::NewEnumerator1, Loop Behavior=::NewEnumerator1, Loop Delay=4.6, Loop Duration=2.3, UseLoopDelay=true)
   &SpawnRate(
    SpawnRate=&FloatFromCurve(
     FloatCurve=DataInterface() {propertyValues {Curve {keys [<at 0 1 RCIM_Cubic>, {arriveTangent=-12.0168, interpMode=RCIM_Cubic, leaveTangent=-12.0168, tangentMode=RCTM_User, time=0.2, value=0}]}}}
     CurveIndex=$Emitter.NormalizedLoopAge
     Scale Curve=56
    )
    Use Spawn Probability=false
   )
   &SpawnRate(
    SpawnGroup=1
    SpawnRate=&FloatFromCurve(
     FloatCurve=DataInterface() {propertyValues {Curve {keys [<at 0 1 RCIM_Cubic>, {arriveTangent=-12.0168, interpMode=RCIM_Cubic, leaveTangent=-12.0168, tangentMode=RCTM_User, time=0.2, value=0}]}}}
     CurveIndex=$Emitter.NormalizedLoopAge
     Scale Curve=56
    )
    Use Spawn Probability=false
   ) {module=SpawnRate001}
  ]
  particleSpawnScript [
   /SetVariables(Particles.VisibilityTag=&SpawnGroup()) {isSetParameters=true, module=SetVariables_#3}
   &InitializeParticle(
    Color=$System.ToD
    Lifetime Max=2.4
    Lifetime Min=1.6
    Lifetime Mode=::NewEnumerator1
    Sprite Rotation Mode=::NewEnumerator2
    Sprite Size Mode=::NewEnumerator1
    Uniform Sprite Size Max=9
    Uniform Sprite Size Min=7
    UsePositionOffset=false
    Write Material Random=true
   )
   &ShapeLocation(Handle Radius=2, Large Radius=4, MaskSpawnGroup=false, Override Seed=false, Shape Primitive=::NewEnumerator3, UseOwnerScale=true)
   &SkeletalMeshLocation(
    Bone / Socket Index=&MergeHelper_IntSwitch(0=2, 1=4, 2=?, 3=?, 4=?, 5=?, Index=&SpawnGroup(), Num=::NewEnumerator0)
    Bone Sampling Mode=::NewEnumerator4
    MaskSpawnGroup=false
    Sampled Position Offset=&AddVectorToPosition(
     Position=$Particles.ShapeLocation.ShapePosition
     Vector=&MakeVector(
      X=0
      Y=0
      Z=&MergeHelper_FloatSwitch(0=3, 1=-3, 2=?, 3=?, Index=&SpawnGroup(), Num=::NewEnumerator0)
     )
    )
    Skeletal Mesh=$System.Skeletal Mesh Reader
   )
   &AddVelocity(
    Cone Angle=60
    Cone Axis=&MakeVector(
     X=&MergeHelper_FloatSwitch(0=1, 1=1, 2=0, 3=0, 4=0, Index=&SpawnGroup(), Num=::NewEnumerator3)
     Y=0
     Z=&MergeHelper_FloatSwitch(0=1, 1=-1, 2=0, 3=0, 4=0, Index=&SpawnGroup(), Num=::NewEnumerator3)
    )
    Module.UseConeFalloff=false
    Module.UseVelDistribution=false
    Override Seed=false
    Velocity Mode=::NewEnumerator2
    Velocity Speed=&MergeHelper_FloatSwitch(
     0=&RandomRangeFloat(Minimum=80, Maximum=120)
     1=&RandomRangeFloat(Minimum=20, Maximum=60)
     2=0
     3=0
     4=0
     Index=&SpawnGroup()
     Num=::NewEnumerator3
    )
   )
   &SubUVAnimation(End Frame=3, SubUV Animation Mode:ENiagara_SubUVLookupMode=::NewEnumerator2)
  ]
  particleUpdateScript [
   &ParticleState()
   &SpriteRotationRate(Rotation Rate=&FloatFromCurve(
     FloatCurve=DataInterface() {propertyValues {Curve {keys [<at 0 1 RCIM_Cubic>, {arriveTangent=-3.1321, interpMode=RCIM_Cubic, leaveTangent=-3.1321, tangentMode=RCTM_User, time=0.6, value=0}]}}}
     Scale Curve=&RandomRangeFloat(Minimum=-2500, Maximum=2500)
    ))
   &ScaleSpriteSize(Uniform Curve Sprite Scale=DataInterface() {propertyValues {Curve {keys [
        {tangentMode=RCTM_User, time=0, value=1}
        {arriveTangent=0.2105, interpMode=RCIM_Cubic, leaveTangent=0.2105, time=0.9, value=1}
        {arriveTangent=-10, interpMode=RCIM_Cubic, leaveTangent=-10, time=0.95, value=1.2}
        <at 1 0 RCIM_Cubic>
       ]}}})
   &GravityForce(Gravity=&MakeVector(
     X=0
     Y=0
     Z=&FloatFromCurve(FloatCurve=DataInterface() {propertyValues {Curve {keys [<at 0.05 0 RCIM_Cubic>,<at 0.1 1 RCIM_Cubic>]}}}, Scale Curve=-980)
    ))
   &Collision(
    Age Colliding Particles=false
    Collision Enabled=&Boolean_AndOperation(
     Bool1=&IsPlatformSetActive(New Platform Set=DataInterface() {propertyValues {Platforms {qualityLevelMask=-8}}})
     Bool2=&Boolean_NotOperation(Bool=$User.bIsFrontEndPreview)
    )
    Control Roll On Collision=true
    Enable Rest State=false
    KillOnCollision=false
    Module.EnableMaxCollisionCount=false
    Module.Randomize Collision Normal Vector=false
    Restitution=0.2
    ScaleRateOfRotation=true
   ) {module=Collision001}
   &Collision(
    Age Colliding Particles=false
    Analytical Collision Plane Position 1:Vector3f=<0 0 -125>
    Analytical Collision Plane Space:ENiagaraCoordinateSpace=::Local
    CPU Collision Type=::NewEnumerator3
    Collision Enabled=&Boolean_OrOperation(
     Bool1=&IsPlatformSetActive(New Platform Set=DataInterface() {propertyValues {Platforms {qualityLevelMask=-28}}})
     Bool2=$User.bIsFrontEndPreview
    )
    Control Roll On Collision=true
    Enable Rest State=false
    KillOnCollision=false
    Module.EnableMaxCollisionCount=false
    Module.Randomize Collision Normal Vector=false
    Particle Radius Scale=5
    Perform 2 Analytical Collision Queries=false
    Restitution=0.2
    ScaleRateOfRotation=false
   ) {module=Collision002}
   &Drag(Drag=&RandomRangeFloat(Minimum=0.8, Maximum=1.2), Use Linear Drag=true, Use Rotational Drag=true)
   &SolveForcesAndVelocity(Clamp Velocity=false, Limit Acceleration=false)
  ]
  rendererClasses [&NiagaraSpriteRendererProperties]
  renderers [
   {rendererClass=&NiagaraSpriteRendererProperties}
   {rendererClass=&NiagaraSpriteRendererProperties, rendererIndex=1}
  ]
 }
}
summary {
 systemName=NS_Emote_SurgeRaven_Candy
 emitters [
  {emitterName=Candy_Head_Rate, rendererClasses [&NiagaraSpriteRendererProperties]}
  {emitterName=Candy_Burst, rendererClasses [&NiagaraSpriteRendererProperties]}
  {emitterName=Candy_Rate001, rendererClasses [&NiagaraSpriteRendererProperties]}
  {emitterName=Candy_Rate002, rendererClasses [&NiagaraSpriteRendererProperties]}
 ]
 userVariables [
  User.bIsFrontEnd {type=NiagaraBool}
  User.bIsFrontEndPreview {type=NiagaraBool}
 ]
}
