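# Niagara slash burst effect for the SurgeRaven player emote, built from four short-lived burst emitters.
# MainSlash and Directional_Streaks take their motion and orientation from the User Velocity and RotationAngle parameters, with sprites aligned to mesh orientation.
# BrightBurst uses three spawn groups to switch color, size and fade across sprite and mesh renderers, while AirStreaks adds small fading SubUV streaks.
# Time-of-day functions adjust color and brightness in all emitters.
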NiagaraSystem /BRCosmetics/Animation/Game/MainPlayer/Emotes/SurgeRaven/Fx/NS_Emote_SurgeRaven_Slash2
dependencies {
 usedDataInterfaces [
  &NiagaraDataInterfaceVector2DCurve
  &NiagaraDataInterfaceCurve
  &NiagaraDataInterfaceVectorCurve
  &NiagaraDataInterfaceColorCurve
 ]
 usedDynamicInputs [&Vector4FromCurve]
 usedModules [&SystemState]
}
emitters {
 AirStreaks {
  emitterUpdateScript [
   &EmitterState()
   &SpawnBurst_Instantaneous(Spawn Count=12, Use Spawn Probability=false, UseLoopCountLimit=false)
  ]
  particleSpawnScript [
   /SetVariables(Particles.RandomFloat=&UniformRangedFloat(Minimum=0, Maximum=3)) {enabled=false, isSetParameters=true}
   &InitializeParticle(
    Color=&MakeLinearColorFromVectorAndFloat(
     "Vector (RGB)"=&MakeVectorFromLinearColorRGB(Color=&Multiply_LinearColorByFloat(
       Linear Color=&RandomRangeLinearColor(Minimum=<rgba 0.325 0.325 0.325 1>, Maximum=<rgba 0.5 0.5 0.5 1>)
       Float=&FNTimeOfDayScalar(Night Scale=0.8, Morning Scale=1, Evening Scale=1)
      ))
     "Float (Alpha)"=1
    )
    Lifetime Max=0.35
    Lifetime Min=0.2
    Lifetime Mode=::NewEnumerator1
    Sprite Size=&Multiply_Vector2D_ByFloat(Float=0.2, Vector2D=&RandomRangeVector2D(Minimum=<50 200>, Maximum=<60 300>))
    Sprite Size Mode=::NewEnumerator4
    UsePositionOffset=false
    Write Material Random=true
   )
   &ShapeLocation(MaskSpawnGroup=false, Override Seed=false, Sphere Radius=16, UseOwnerScale=true)
   &AddVelocity(Module.Use Falloff Distance=false, Override Seed=false, Velocity Mode=::NewEnumerator1)
   &ApplyInitialForces()
   &SubUVAnimation(End Frame=3, SubUV Animation Mode:ENiagara_SubUVLookupMode=::NewEnumerator2)
  ]
  particleUpdateScript [
   &ParticleState()
   &SolveForcesAndVelocity(Clamp Velocity=false, Limit Acceleration=false)
   &ScaleSpriteSize(Scale Factor=&Vector2DFromCurve(
     Vector2Curve=DataInterface() {propertyValues {XCurve {keys [<at 0.0036 1.6897>,<at 0.9967 0.0582>]}, YCurve {keys [<at 0 1>,<at 1.002 3.7227>]}}}
     ScaleCurve=&Multiply_Vector2D_ByFloat(Float=0.4, Vector2D=<1 1>)
    ))
   &ScaleColor(
    Scale Alpha=&FloatFromCurve(
     FloatCurve=DataInterface() {propertyValues {Curve {keys [
         {interpMode=RCIM_Cubic, leaveTangent=-1.9294, tangentMode=RCTM_Break, time=0, value=1}
         {interpMode=RCIM_Cubic, tangentMode=RCTM_Break, time=1, value=0}
        ]}}}
     Scale Curve=&RandomRangeFloat(Minimum=0.1, Maximum=0.3)
    )
    ScaleA=true
    ScaleRGB=false
   )
   &DynamicMaterialParameters(
    Index 0 Param 2=1
    Index 0 Param 3=1
    Index 0 Param 4=&FloatFromCurve(FloatCurve=DataInterface() {propertyValues {Curve {keys [<at 0.3 1>,<at 1 0.25>]}}})
    Module.Param0WriteEnabled=true
    Module.Param1WriteEnabled=true
    Module.Param2WriteEnabled=true
    Module.Param3WriteEnabled=true
   )
  ]
  rendererClasses [&NiagaraSpriteRendererProperties]
  renderers [{rendererClass=&NiagaraSpriteRendererProperties}]
 }
 BrightBurst {
  emitterUpdateScript [
   &EmitterState()
   &SpawnBurst_Instantaneous(Use Spawn Probability=false, UseLoopCountLimit=false)
   &SpawnBurst_Instantaneous(Spawn Group=1, Use Spawn Probability=false, UseLoopCountLimit=false) {module=SpawnBurst_Instantaneous001}
   &SpawnBurst_Instantaneous(Spawn Group=2, Use Spawn Probability=false, UseLoopCountLimit=false) {module=SpawnBurst_Instantaneous002}
  ]
  particleSpawnScript [
   /SetVariables(Particles.VisibilityTag=&SpawnGroup()) {isSetParameters=true}
   &InitializeParticle(
    Color=&MergeHelper_ColorSwitch(
     0=&FN_Select_Color_from_Time_Of_Day(
      Morning=<rgba 0.3125 0.3125 0.3125 1>
      Day=<rgba 0.4601 0.4903 0.5104 1>
      Evening=<rgba 0.3125 0.3125 0.3125 1>
      Night=<rgba 0.1648 0.2191 0.3229 1>
     )
     1=<rgba 0.0104 0.0104 0.0104 50>
     2=<rgba 0.0104 0.0104 0.0104 5>
     3=?
     Index=&SpawnGroup()
     Num=::NewEnumerator1
    )
    Lifetime=&MergeHelper_FloatSwitch(0=0.3, 1=0.15, 2=0.12, 3=?, Index=&SpawnGroup(), Num=::NewEnumerator1)
    Mesh Scale Mode=::NewEnumerator3
    Mesh Uniform Scale=&MakeFloatFromVector(Vector=&Multiply_VectorByFloat(Vector=<1.5 1.5 1.5>, Float=0.5), Channel=::NewEnumerator0)
    Sprite Size Mode=::NewEnumerator3
    Uniform Sprite Size=&MergeHelper_FloatSwitch(0=85, 1=0, 2=115, 3=?, Index=&SpawnGroup(), Num=::NewEnumerator1)
    UsePositionOffset=false
    Write Material Random=true
   )
   &CameraOffset(Camera Offset Amount=15)
  ]
  particleUpdateScript [
   &ParticleState()
   &ScaleSpriteSize(
    Scale Sprite Size Mode=::NewEnumerator0
    Uniform Scale Factor=&MergeHelper_FloatSwitch(
     0=1
     1=0
     2=&MakeFloatFromVector2d(Vector 2D=&Vector2DFromFloat(Value=&FloatFromCurve(FloatCurve=DataInterface() {propertyValues {Curve {keys [{tangentMode=RCTM_User, time=0, value=0.4}, {tangentMode=RCTM_User, time=1, value=0.8}]}}})))
     3=?
     Index=&SpawnGroup()
     Num=::NewEnumerator1
    )
   )
   &ScaleMeshSize(Scale Factor=&VectorFromCurve(VectorCurve=DataInterface() {propertyValues {XCurve {keys [<at 0 0>,<at 0.25 1>]}, YCurve {keys [<at 0 0>,<at 0.25 1>]}, ZCurve {keys [<at 0 0>,<at 0.25 1>]}}}))
   &ScaleColor(
    Scale Alpha=&MergeHelper_FloatSwitch(
     0=&FloatFromCurve(FloatCurve=DataInterface() {propertyValues {Curve {keys [
          {arriveTangent=51.5483, interpMode=RCIM_Cubic, leaveTangent=51.5483, tangentMode=RCTM_User, time=0, value=0}
          {arriveTangent=-0.1074, interpMode=RCIM_Cubic, leaveTangent=-0.1074, time=0.0647, value=1}
          {arriveTangentWeight=0.0616, interpMode=RCIM_Cubic, leaveTangentWeight=0.3948, tangentMode=RCTM_User, time=0.283, value=-0.0304}
          <at 1 0>
         ]}}})
     1=&FloatFromCurve(FloatCurve=DataInterface() {propertyValues {Curve {keys [<at 0 1>,<at 1 0>]}}})
     2=&FloatFromCurve(FloatCurve=DataInterface() {propertyValues {Curve {keys [<at 0 1 RCIM_Cubic>,<at 1 0 RCIM_Cubic>]}}})
     3=?
     Index=&SpawnGroup()
     Num=::NewEnumerator1
    )
    ScaleA=true
    ScaleRGB=true
   )
   /SetVariables(Particles.BurstMeshDP:Vector4f=&MakeVector4(
     X=&FloatFromCurve(FloatCurve=DataInterface() {propertyValues {Curve {keys [{arriveTangent=0.0592, interpMode=RCIM_Cubic, time=0, value=0}, <at 1 1 RCIM_Cubic>]}}})
     Y=50
     Z=1
     W=1
    )) {isSetParameters=true}
  ]
  rendererClasses [&NiagaraSpriteRendererProperties, &NiagaraMeshRendererProperties]
  renderers [
   {rendererClass=&NiagaraSpriteRendererProperties}
   {rendererClass=&NiagaraMeshRendererProperties, rendererIndex=1}
   {rendererClass=&NiagaraSpriteRendererProperties, rendererIndex=2}
  ]
 }
 Directional_Streaks {
  emitterUpdateScript [
   &EmitterState()
   &SpawnBurst_Instantaneous(Spawn Time=0.02, Use Spawn Probability=true, UseLoopCountLimit=false)
  ]
  particleSpawnScript [
   &InitializeParticle(
    Lifetime=0.15
    Mesh Scale Max=<1 2 2>
    Mesh Scale Mode=::NewEnumerator2
    Sprite Rotation Angle=$User.RotationAngle
    Sprite Rotation Mode=::NewEnumerator0
    Sprite Size Max=<3 65>
    Sprite Size Min=<2 40>
    Sprite Size Mode=::NewEnumerator2
    UsePositionOffset=false
    Write Material Random=true
   ) {module=InitializeParticle001}
   &AddVelocity(Velocity=$User.Velocity, Velocity Speed Scale=2.65)
   &CameraOffset()
   /SetVariables(Particles.Size Blurry Streak:Vector2f=<20 150>) {isSetParameters=true}
   /SetVariables(Particles.Color Dark Streak:LinearColor=<rgba 0.2344 0.2344 0.2344 0.5>, Particles.Size Dark Streak:Vector2f=<5 100>) {isSetParameters=true}
   &SubUVAnimation(End Frame=3, SubUV Animation Mode:ENiagara_SubUVLookupMode=::NewEnumerator2)
   &SpriteFacingAndAlignment(Sprite Alignment=<1 0 0>, Sprite Facing=<0 1 0>, WriteAlignment=true, WriteFacing=true) {enabled=false}
   &InitialMeshOrientation(
    Mesh Axis To Orient=<0 1 0>
    Mesh Orientation Mode=::NewEnumerator2
    Rotation=&MakeVector(X=$User.RotationAngle, Y=0, Z=0)
    Use Rotation Vector=true
   )
  ]
  particleUpdateScript [
   &ParticleState()
   &SolveForcesAndVelocity(Clamp Velocity=false, Limit Acceleration=false)
   &AlignSpriteToMeshOrientation(Mesh Orientation Relative Sprite Alignment Vector=<0 0 1>, SetAlignment=true, SetFacing=true)
   &ScaleSpriteSize(Scale Factor=&Vector2DFromCurve(
     Vector2Curve=DataInterface() {propertyValues {XCurve {keys [<at 0 1 RCIM_Cubic>,<at 1 0 RCIM_Cubic>]}, YCurve {keys [<at 0 0.5 RCIM_Cubic>,<at 1 2 RCIM_Cubic>]}}}
     ScaleCurve=&Multiply_Vector2D_ByFloat(Float=1, Vector2D=<1.5 1>)
    ))
   &ScaleColor(
    Scale Alpha=&FloatFromCurve(FloatCurve=DataInterface() {propertyValues {Curve {keys [<at 0 1 RCIM_Cubic>,<at 1 0>]}}}, Scale Curve=0.15)
    Scale RGB=&VectorFromFloat(Value=&FloatFromCurve(
      FloatCurve=DataInterface() {propertyValues {Curve {keys [<at 0 10 RCIM_Cubic>,<at 1 1 RCIM_Cubic>]}}}
      Scale Curve=&Multiply_Float(
       A=0.5
       B=&FNTimeOfDayScalar(Night Scale=0.25, Morning Scale=0.45, Day Scale=0.8, Evening Scale=0.45)
      )
     ))
    ScaleA=true
    ScaleRGB=true
   )
   /SetVariables(
    Particles.Color Dark Streak:LinearColor=&MakeLinearColorFromVectorAndFloat(
     "Vector (RGB)"=&MakeVectorFromLinearColorRGB(Color=$Particles.Initial.Color Dark Streak)
     "Float (Alpha)"=&FloatFromCurve(
      FloatCurve=DataInterface() {propertyValues {Curve {keys [{interpMode=RCIM_Cubic, tangentMode=RCTM_User, time=0, value=1}, {interpMode=RCIM_Cubic, tangentMode=RCTM_User, time=0.5, value=0}]}}}
      Scale Curve=0.5
     )
    )
    Particles.Size Dark Streak:Vector2f=&Multiply_Vector2D(
     A=$Particles.Initial.Size Dark Streak
     B=&MakeVector2D(
      X=1
      Y=&FloatFromCurve(
       FloatCurve=DataInterface() {propertyValues {Curve {keys [
           {arriveTangent=1.3254, interpMode=RCIM_Cubic, leaveTangent=1.3254, tangentMode=RCTM_User, time=0, value=0}
           {interpMode=RCIM_Cubic, tangentMode=RCTM_User, time=1, value=1}
          ]}}}
       Scale Curve=0.5
      )
     )
    )
   ) {isSetParameters=true}
  ]
  rendererClasses [&NiagaraSpriteRendererProperties, &NiagaraMeshRendererProperties]
  renderers [
   {rendererClass=&NiagaraSpriteRendererProperties}
   {rendererClass=&NiagaraSpriteRendererProperties, rendererIndex=1}
   {rendererClass=&NiagaraSpriteRendererProperties, rendererIndex=2}
   {rendererClass=&NiagaraMeshRendererProperties, rendererIndex=3}
  ]
 }
 MainSlash {
  emitterUpdateScript [
   &EmitterState(Life Cycle Mode=::NewEnumerator1, Loop Behavior=::NewEnumerator1, Loop Duration=1, UseLoopDelay=false)
   &SpawnBurst_Instantaneous(Spawn Count=1, Use Spawn Probability=true)
  ]
  particleSpawnScript [
   &InitializeParticle(
    Color=<rgba 2 2 2 0.2>
    Lifetime=0.225
    Mesh Scale=<0.3 0.05 1>
    Mesh Scale Mode=::NewEnumerator4
    Position Offset Coordinate Space=::Local
    Sprite Rotation Angle=0.4401
    Sprite Rotation Mode=::NewEnumerator0
    Sprite Size Max=<3 175>
    Sprite Size Min=<1 100>
    Sprite Size Mode=::NewEnumerator2
    UsePositionOffset=true
    Write Material Random=true
   ) {module=InitializeParticle001}
   &AddVelocity(Velocity=$User.Velocity, Velocity Speed Scale=4.25)
   &CameraOffset(Camera Offset Amount=20)
   &DynamicMaterialParameters(
    Index 0 Param 1=1
    Index 0 Param 2=1
    Module.Param0WriteEnabled=true
    Module.Param1WriteEnabled=true
    Module.Param2WriteEnabled=true
    Module.Param3WriteEnabled=true
   )
   &InitialMeshOrientation(
    Mesh Axis To Orient=<0 1 0>
    Mesh Orientation Mode=::NewEnumerator2
    Rotation=&MakeVector(X=$User.RotationAngle, Y=0, Z=0)
    Use Rotation Vector=true
   )
  ]
  particleUpdateScript [
   &ParticleState()
   &SolveForcesAndVelocity(Clamp Velocity=false, Limit Acceleration=false)
   &AlignSpriteToMeshOrientation(Mesh Orientation Relative Sprite Alignment Vector=<0 0 1>, SetAlignment=true, SetFacing=true)
   &ScaleColor(
    Color Value To Scale=$Particles.Color
    Scale Alpha=&FloatFromCurve(FloatCurve=DataInterface() {propertyValues {Curve {keys [
         {interpMode=RCIM_Cubic, tangentMode=RCTM_User, time=0, value=1}
         {arriveTangent=-3.4409, interpMode=RCIM_Cubic, leaveTangent=-3.4409, tangentMode=RCTM_User, time=1, value=0}
        ]}}})
    Scale RGB=&Multiply_VectorByFloat(
     Vector=&MakeVectorFromLinearColorRGB(Color=&ColorFromCurve(ColorCurve=DataInterface() {propertyValues {
         AlphaCurve {keys [<at 0 1 RCIM_Cubic>]}
         BlueCurve {keys [<at 0 0.5>,<at 0.2843 1.25 RCIM_Cubic>,<at 0.5346 0.5 RCIM_Cubic>]}
         GreenCurve {keys [<at 0 0.5>,<at 0.2843 1.25 RCIM_Cubic>,<at 0.5346 0.5 RCIM_Cubic>]}
         RedCurve {keys [<at 0 0.5>,<at 0.2843 1.25 RCIM_Cubic>,<at 0.5346 0.5 RCIM_Cubic>]}
        }}))
     Float=&FNTimeOfDayScalar(Night Scale=0.7, Morning Scale=0.7, Day Scale=0.7, Evening Scale=0.7)
    )
    ScaleA=true
    ScaleRGB=true
   )
   &ScaleSpriteSize(
    Non-Uniform Curve Sprite Scale=DataInterface() {propertyValues {
      XCurve {keys [<at 0 0 RCIM_Cubic>, {arriveTangent=-0.1, interpMode=RCIM_Cubic, leaveTangent=-0.1, time=0.2, value=1.075}, <at 1 -0.1 RCIM_Cubic>]}
      YCurve {keys [
        <at 0 0 RCIM_Cubic>
        {arriveTangent=0.0607, interpMode=RCIM_Cubic, leaveTangent=0.0607, tangentMode=RCTM_User, time=0.1, value=1.025}
        <at 1 0.8 RCIM_Cubic>
       ]}
     }}
    Scale Sprite Size Mode=::NewEnumerator3
   ) {module=ScaleSpriteSize001}
   &FN_Sprite_Attributes(
    Color=<rgba 0 0 0 1>
    Create Color Attribute=true
    Create Sprite Rotation Attribute=true
    Create SpriteSize Attribute=true
    SpriteRotation:NiagaraFloat=$User.RotationAngle
    SpriteSize=&Multiply_Vector2D(A=$Particles.SpriteSize, B=<1.1 1.25>)
    Write to Module Namespace=true
   )
  ]
  rendererClasses [&NiagaraSpriteRendererProperties]
  renderers [
   {rendererClass=&NiagaraSpriteRendererProperties}
   {rendererClass=&NiagaraSpriteRendererProperties, rendererIndex=1}
  ]
 }
}
summary {
 systemName=NS_Emote_SurgeRaven_Slash2
 emitters [
  {emitterName=AirStreaks, rendererClasses [&NiagaraSpriteRendererProperties]}
  {emitterName=BrightBurst, rendererClasses [&NiagaraSpriteRendererProperties, &NiagaraMeshRendererProperties]}
  {emitterName=MainSlash, rendererClasses [&NiagaraSpriteRendererProperties]}
  {
   emitterName=Directional_Streaks
   rendererClasses [&NiagaraSpriteRendererProperties, &NiagaraMeshRendererProperties]
  }
 ]
 userVariables [
  User.bIsFrontEnd {type=NiagaraBool}
  User.bIsFrontEndPreview {type=NiagaraBool}
  User.OwnerLocalVelocity {type=Vector3f, description="", defaultValue {struct=Vector3f, value=<0 0 0>}}
  User.RotationAngle {type=NiagaraFloat, defaultValue=0, description=""}
  User.Velocity {type=Vector3f, description="", defaultValue {struct=Vector3f, value=<0 0 0>}}
 ]
}
