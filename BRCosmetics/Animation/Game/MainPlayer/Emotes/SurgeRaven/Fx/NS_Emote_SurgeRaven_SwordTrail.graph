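# Niagara system that draws a sword trail for the SurgeRaven player emote.
# A single FN_AnimTrail emitter spawns ribbon particles from bone or socket indices, using a socket reader interface, and feeds two ribbon renderers.
# Particles are very short-lived, with a dim grey color that varies by time of day.
# A second bone-pair initializer and an added velocity are present but disabled.
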NiagaraSystem /BRCosmetics/Animation/Game/MainPlayer/Emotes/SurgeRaven/Fx/NS_Emote_SurgeRaven_SwordTrail
dependencies {
 usedDataInterfaces [&NiagaraDataInterfaceSocketReader]
 usedDynamicInputs []
 usedModules [
  &self:SystemSpawnScript.SystemScriptSource.SystemScriptGraph.NiagaraNodeAssignment_0.SetVariables_#1
  &SystemState
 ]
}
emitters {FN_AnimTrail {
  emitterSpawnScript [&FN_RibbonIDOffsetUtility()]
  emitterUpdateScript [
   &EmitterState()
   &FN_SpawnAnimTrail()
  ]
  particleSpawnScript [
   &FN_InitializeAnimTrailBoneIndecies()
   &FN_InitializeAnimTrailBoneIndecies(AnimTrail Selection=1, First Bone or Socket Index=2, Second Bone or Socket Index=3) {enabled=false, module=FN_InitializeAnimTrailBoneIndecies001}
   &FN_InitializeAnimTrail(
    Color=&FNTimeOfDayColor(
     Night Color=<rgba 0.0573 0.0573 0.0573 1>
     Morning Color=<rgba 0.099 0.099 0.099 1>
     Day Color=<rgba 0.1667 0.1667 0.1667 1>
     Evening Color=<rgba 0.099 0.099 0.099 1>
    )
    Lifetime=0.1
   )
   &AddVelocity(Velocity=<0 400 0>) {enabled=false}
  ]
  particleUpdateScript [
   &SolveForcesAndVelocity(Clamp Velocity=false, Limit Acceleration=false)
   &FN_UpdateAnimTrail() {module=FN_UpdateAnimTrail001}
  ]
  rendererClasses [&NiagaraRibbonRendererProperties]
  renderers [
   {rendererClass=&NiagaraRibbonRendererProperties}
   {rendererClass=&NiagaraRibbonRendererProperties, rendererIndex=1}
  ]
 }}
summary {
 systemName=NS_Emote_SurgeRaven_SwordTrail
 emitters [{emitterName=FN_AnimTrail, rendererClasses [&NiagaraRibbonRendererProperties]}]
 userVariables [User.RibbonIDOffset {type=NiagaraInt32}]
}
